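# Translucent master material that renders Niagara FLIP fluid as a screen-space surface from simulation-fed depth, normal and thickness textures.
# Jittered radial-blur Custom nodes smooth those buffers, which drive opacity, refracted scene color, reflections, fake SSS and optional thin-film tint.
# An optional Custom ray march against fluid depth adds contact shadows, fading step count with distance.
# A BakeNormalAndThickness switch supports baking, with quadrant masks packing normals, motion vectors and other channels.
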Material /Game/Effects/FlipSimulation/Rendering/M_FluidMaterial
graph {
 properties {
  bDisableDepthTest=true
  bEnableResponsiveAA=true
  bTangentSpaceNormal=false
  bUsedWithInstancedStaticMeshes=true
  bUsedWithNanite=true
  bUsedWithNiagaraMeshParticles=true
  bUsedWithNiagaraSprites=true
  blendMode=BLEND_Translucent
  pixelDepthOffsetMode=PDOM_Legacy
  translucencyPass=MTP_BeforeDOF
  twoSided=true
  usedShadingModels=Default Lit
 }
 nodes [
  n1 = TextureSample(Tex=TextureObjectParameter() {group=FromNiagara, parameterName=RenderValid}) {constCoordinate=1}
  Multiply(
   A=Multiply(A=OneMinus(None=@n1.R)) {constB=-100000000}
   B=MaterialFunctionCall() {materialFunction=&CameraDirectionVector}.Result
  )
  NamedRerouteDeclaration(None=TextureObject() {samplerType=SAMPLERTYPE_LinearGrayscale, texture=/Engine/EngineMaterials/Good64x64TilingNoiseHighFreq}) {name=RandomTex, nodeColor=<rgba 0 0.5078 1 1>}
  MaterialExpressionTextureObjectParameter_1 = TextureObjectParameter() {group=FromNiagara, parameterName=TexBack}
  MaterialExpressionStaticSwitchParameter_12 = StaticSwitchParameter(True=TextureCoordinate() {uTiling=2, vTiling=2}, False=TextureCoordinate()) {parameterName=BakeNormalAndThickness}
  MaterialExpressionMultiply_10 = Multiply(A=ScalarParameter() {defaultValue=0.007, group=Thickness, parameterName=Thickness_BlurDistance}, B=NamedRerouteUsage())
  MaterialExpressionScalarParameter_3 = ScalarParameter() {defaultValue=6, group=Thickness, parameterName=Thickness_DistanceSteps}
  MaterialExpressionScalarParameter_6 = ScalarParameter() {defaultValue=4, group=Thickness, parameterName=Thickness_RadialSteps}
  MaterialExpressionConstant_1 = Constant() {r=0.618}
  MaterialExpressionScalarParameter_10 = ScalarParameter() {defaultValue=0.1, group=Thickness, parameterName=Thickness_KernelPower}
  MaterialExpressionNamedRerouteUsage_0 = NamedRerouteUsage()
  MaterialExpressionConstant_13 = Constant()
  MaterialExpressionScalarParameter_31 = ScalarParameter() {defaultValue=0.5, group=Thickness, parameterName=Thickness_JitterIntensity}
  MaterialExpressionConstant_11 = Constant() {r=1}
  n2 = Custom!(
   Tex=@MaterialExpressionTextureObjectParameter_1
   UV=@MaterialExpressionStaticSwitchParameter_12
   Distance=@MaterialExpressionMultiply_10
   DistanceSteps=@MaterialExpressionScalarParameter_3
   RadialSteps=@MaterialExpressionScalarParameter_6
   RadialOffset=@MaterialExpressionConstant_1
   KernelPower=@MaterialExpressionScalarParameter_10
   RandomTex=@MaterialExpressionNamedRerouteUsage_0
   JitterRandomIntensity=@MaterialExpressionConstant_13
   JitterIntensity=@MaterialExpressionScalarParameter_31
   ClampSmpResult=@MaterialExpressionConstant_11
  ) {
   code="\r\nfloat4 CurColor=0;\r\nfloat2 NewUV = UV;\r\nint i=0;\r\nfloat StepSize = Distance / (int) DistanceSteps;\r\nfloat CurDistance=0;\r\nfloat2 CurOffset=0;\r\nfloat SubOffset = 0;\r\nfloat TwoPi = 6.283185;\r\nfloat accumdist=0;\r\nfloat2 ScreenUV = MaterialFloat2(ScreenAlignedPosition(Parameters.ScreenPosition).xy);\r\nfloat2 ScenePixels=View.BufferSizeAndInvSize.xy*ScreenUV;\r\nScenePixels+=View.TemporalAAParams.r;\r\nfloat2 RandomSamp = ((uint)(ScenePixels.x) + 2 * (uint)(ScenePixels.y)) % 5;\r\n//RandomSamp+=Texture2DSample(RandomTex,RandomTexSampler,ScenePixels)*JitterRandomIntensity;\r\nRandomSamp/=5;\r\nRandomSamp-=0.5;\r\nRandomSamp*=JitterIntensity;\r\nfloat TempAARotation = 1;\r\nfloat TempAADistance = 1;\r\nTempAARotation*=RandomSamp.x;\r\nTempAADistance*=StepSize*RandomSamp.x;\r\n\r\nif (DistanceSteps < 1)\r\n\r\n{\r\nreturn Texture2DSample(Tex,TexSampler,UV);\t\t\r\n}\r\nelse\r\n{\r\nwhile ( i < (int) DistanceSteps)\r\n{\r\n\tCurDistance += StepSize;\r\n\tfor (int j = 0; j < (int) RadialSteps; j++)\r\n\t{\r\n\t\tSubOffset +=1;\r\n\t\tCurOffset.x = cos(TwoPi*((TempAARotation+SubOffset) / RadialSteps));\r\n\t\tCurOffset.y = sin(TwoPi*((TempAARotation+SubOffset) / RadialSteps));\r\n\t\tNewUV.x = UV.x + CurOffset.x * (CurDistance+(RandomSamp*TempAADistance));\r\n\t\tNewUV.y = UV.y + CurOffset.y * (CurDistance+(RandomSamp*TempAADistance));\r\n\t\tfloat distpow = pow(CurDistance, KernelPower);\r\n\t\tfloat4 tmpTexSmp = Texture2DSample(Tex,TexSampler,NewUV);\r\n\t\tCurColor += lerp(tmpTexSmp,clamp(tmpTexSmp,0,100000),ClampSmpResult)*distpow;\t\t\r\n\t\taccumdist += distpow;\r\n\t}\r\nSubOffset +=RadialOffset;\r\ni++;\r\n}\r\nCurColor = CurColor;\r\nCurColor /=accumdist;\r\nreturn CurColor;\r\n}\r\n\r\n"
   outputType=CMOT_Float4
   inputs [
    {inputName=Tex, input {expression=&self:MaterialExpressionTextureObjectParameter_1}}
    {inputName=UV, input {expression=&self:MaterialExpressionStaticSwitchParameter_12}}
    {inputName=Distance, input {expression=&self:MaterialExpressionMultiply_10}}
    {inputName=DistanceSteps, input {expression=&self:MaterialExpressionScalarParameter_3}}
    {inputName=RadialSteps, input {expression=&self:MaterialExpressionScalarParameter_6}}
    {inputName=RadialOffset, input {expression=&self:MaterialExpressionConstant_1}}
    {inputName=KernelPower, input {expression=&self:MaterialExpressionScalarParameter_10}}
    {inputName=RandomTex, input {expression=&self:MaterialExpressionNamedRerouteUsage_0}}
    {inputName=JitterRandomIntensity, input {expression=&self:MaterialExpressionConstant_13}}
    {inputName=JitterIntensity, input {expression=&self:MaterialExpressionScalarParameter_31}}
    {inputName=ClampSmpResult, input {expression=&self:MaterialExpressionConstant_11}}
   ]
  }
  n3 = ComponentMask(None=@n2) {r=true}
  n4 = Saturate(None=ScalarParameter() {group=FromNiagara, parameterName=OrthogonalVelocityMult})
  n5 = Divide(A=Divide(A=@n3, B=Add(A=@n4)), B=ScalarParameter() {defaultValue=100, group=Thickness, parameterName=Thickness_Divide}) {constB=250}
  NamedRerouteDeclaration(None=Power(Base=Saturate(None=@n5), Exp=ScalarParameter() {defaultValue=1, group=Thickness, parameterName=Thickness_ResultPower})) {name=ThicknessOutput, nodeColor=<rgba 0 0.9766 1 1>}
  n6 = MaterialFunctionCall!(Center Point World Position=ActorPositionWS()) {materialFunction=/Engine/Functions/Engine_MaterialFunctions02/UVs/WorldSpaceAlignedScreenCoordinates}
  n7 = ScalarParameter() {defaultValue=10, group=ThinFilm, parameterName=ThinFilm_NormalStrengthA}
  n8 = ScalarParameter() {defaultValue=40, group=ThinFilm, parameterName=ThinFilm_NormalStrengthB}
  n9 = NamedRerouteUsage()
  n10 = Multiply(A=CameraVectorWS())
  n11 = NamedRerouteUsage()
  n12 = Constant() {r=1.03}
  n13 = Add(
   A=Multiply(
    A=LinearInterpolate(A=@n7, B=@n8, Alpha=@n9)
    B=MaterialFunctionCall!(Ray Direction=@n10, Surface Normal=@n11, Refractive Index Origin, Refractive Index Target=@n12) {materialFunction=/Engine/Functions/Engine_MaterialFunctions01/Vectors/Refract}.Vector
   )
   B=WorldPosition().XYZ
  )
  n14 = CameraPositionWS()
  n15 = Subtract(
   A=MaterialFunctionCall!(
    Camera Relative Position=Subtract(A=@n13, B=@n14)
    X Camera to World Vector
    Y Camera to World Vector
    Z Camera to World Vector
    "Tan(FOV/2)*[1,Screen Res Y/X]"
   ) {materialFunction=/Engine/Functions/Engine_MaterialFunctions02/Math/ManualWorldToScreenUVsTransform}.Screen UVs
   B=ScreenPosition().ViewportUV
  )
  n16 = TextureSampleParameter2D(UVs=Multiply(A=Add(A=@n6.Offset and Scaled UVS, B=@n15), B=ScalarParameter() {defaultValue=2, group=ThinFilm, parameterName=ThinFilm_TextureTile})) {
   group=ThinFilm
   parameterName=ThinFilm_Texture
   samplerType=SAMPLERTYPE_LinearColor
   texture=/Game/Athena/Apollo/Water/Textures/02_Rounded/03_WaterFoam_Veroni_base_H
  }
  n17 = ScalarParameter() {defaultValue=-1, group=ThinFilm, parameterName=ThinFilm_HueShiftStr}
  n18 = Constant3Vector() {constant=<rgba 1 0 0 0>}
  n19 = ScalarParameter() {defaultValue=1, group=ThinFilm, parameterName=ThinFilm_Source}
  n20 = LinearInterpolate(
   A=VectorParameter() {group=BaseColor, parameterName=BaseColor_LowThickness, defaultValue=<rgba 0.9 0.9 0.9 1>}.RGB
   B=VectorParameter() {group=BaseColor, parameterName=BaseColor_HighThickness, defaultValue=<rgba 0.8 0.8 0.8 1>}.RGB
   Alpha=NamedRerouteUsage()
  )
  n21 = VectorParameter() {group=ThinFilm, parameterName=ThinFilm_Fresnel, channelNames {a="", b="", g=Base, r=Exp}, defaultValue=<rgba 2 0.2 0 1>}
  n22 = Multiply(
   A=MaterialFunctionCall(Hue Shift Percentage=Multiply(A=@n16.R, B=@n17), Texture=LinearInterpolate(A=@n20, B=@n18, Alpha=@n19)) {materialFunction=&HueShift}.Result
   B=Fresnel(ExponentIn=@n21.Exp, BaseReflectFractionIn=@n21.Exp, Normal=NamedRerouteUsage())
  )
  n23 = ScalarParameter() {defaultValue=1, group=ThinFilm, parameterName=ThinFilm_Str}
  n24 = StaticSwitchParameter(True=Add(A=Multiply(A=@n22, B=@n23), B=@n20), False=@n20) {parameterName=ThinFilm}
  n25 = NamedRerouteUsage()
  n26 = ScalarParameter() {group=Opacity, parameterName=OpaqueFluid}
  NamedRerouteDeclaration(None=LinearInterpolate(A=Multiply(A=@n24, B=@n25), Alpha=@n26) {constB=0}) {name=ColoredSceneColor, nodeColor=<rgba 0.125 1 0 1>}
  n27 = VectorParameter() {group=BaseColor, parameterName=BaseColorLighting, channelNames {a="", b="", g=LightFace, r=DarkFace}, defaultValue=<rgba 1 1 0 0>}
  n28 = NamedRerouteUsage()
  n29 = VectorParameter() {group=BaseColor, parameterName=HueShiftControl, channelNames {a="", b="", g=LightFace, r=DarkFace}}
  n30 = LinearInterpolate(A=@n29.DarkFace, B=@n29.DarkFace, Alpha=NamedRerouteUsage())
  n31 = LinearInterpolate(B=@n24, Alpha=@n26) {constB=0}
  NamedRerouteDeclaration(None=Multiply(
    A=LinearInterpolate(A=@n27.DarkFace, B=@n27.DarkFace, Alpha=@n28)
    B=MaterialFunctionCall(Hue Shift Percentage=@n30, Texture=@n31) {materialFunction=&HueShift}.Result
   )) {name=BaseColor, nodeColor=<rgba 0 0.2031 1 1>}
  MaterialExpressionWorldPosition_4 = WorldPosition() {worldPositionShaderOffset=WPT_ExcludeAllShaderOffsets}
  MaterialExpressionTextureObjectParameter_4 = TextureObjectParameter() {group=FromNiagara, parameterName=Tex}
  MaterialExpressionSkyAtmosphereLightDirection_3 = SkyAtmosphereLightDirection()
  MaterialExpressionScalarParameter_8 = ScalarParameter() {defaultValue=35, group=ContactShadow, parameterName=ContactShadow_RayMarchingDistance}
  MaterialExpressionTextureCoordinate_2 = TextureCoordinate()
  MaterialExpressionCameraPositionWS_5 = CameraPositionWS()
  MaterialExpressionMaterialFunctionCall_13 = MaterialFunctionCall() {materialFunction=&WorldPositionBehindTranslucency}
  MaterialExpressionTransform_3 = Transform(None=Constant3Vector() {constant=<rgba 1 0 0 1>}) {transformSourceType=TRANSFORMSOURCE_Camera}
  MaterialExpressionTransform_4 = Transform(None=Constant3Vector() {constant=<rgba 0 1 0 1>}) {transformSourceType=TRANSFORMSOURCE_Camera}
  MaterialExpressionTransform_5 = Transform(None=Constant3Vector() {constant=<rgba 0 0 1 1>}) {transformSourceType=TRANSFORMSOURCE_Camera}
  MaterialExpressionViewProperty_0 = ViewProperty() {property=MEVP_TanHalfFieldOfView}
  MaterialExpressionScreenPosition_0 = ScreenPosition()
  MaterialExpressionScalarParameter_13 = ScalarParameter() {defaultValue=60, group=ContactShadow, parameterName=ContactShadow_MaxDistance}
  n32 = VectorParameter() {group=FromNiagara, parameterName=WorldToUnit_Row0}
  MaterialExpressionAppendVector_0 = AppendVector(A=@n32.RGB, B=@n32.RGB)
  n33 = VectorParameter() {group=FromNiagara, parameterName=WorldToUnit_Row1}
  MaterialExpressionAppendVector_1 = AppendVector(A=@n33.RGB, B=@n33.RGB)
  n34 = VectorParameter() {group=FromNiagara, parameterName=WorldToUnit_Row2}
  MaterialExpressionAppendVector_2 = AppendVector(A=@n34.RGB, B=@n34.RGB)
  n35 = VectorParameter() {group=FromNiagara, parameterName=WorldToUnit_Row3}
  MaterialExpressionAppendVector_3 = AppendVector(A=@n35.RGB, B=@n35.RGB)
  MaterialExpressionMaterialFunctionCall_19 = MaterialFunctionCall() {materialFunction=&CameraDirectionVector}
  MaterialExpressionNamedRerouteUsage_1 = NamedRerouteUsage()
  n36 = VectorParameter() {
   group=ContactShadow
   parameterName=ContactShadow_ControlPerformance
   channelNames {a="", b="", g=DisableStepCountDistance, r=MaxStepCountDistance}
   defaultValue=<rgba 500 2000 0 1>
  }
  n37 = Subtract(A=Distance(A=@MaterialExpressionMaterialFunctionCall_13.Result, B=CameraPositionWS()), B=@n36.MaxStepCountDistance)
  MaterialExpressionLinearInterpolate_10 = LinearInterpolate(
   A=ScalarParameter() {defaultValue=5, group=ContactShadow, parameterName=ContactShadow_StepCount}
   Alpha=Saturate(None=Divide(A=@n37, B=@n36.DisableStepCountDistance))
  ) {constB=0}
  MaterialExpressionScalarParameter_19 = ScalarParameter() {defaultValue=0.5, group=ContactShadow, parameterName=ContactShadow_StepDistanceJitterFactor}
  MaterialExpressionScalarParameter_20 = ScalarParameter() {defaultValue=0.5, group=ContactShadow, parameterName=ContactShadow_LightDirJitterAngle}
  MaterialExpressionScalarParameter_23 = ScalarParameter() {defaultValue=0.85, group=ContactShadow, parameterName=ContactShadow_DepthFixOffsetFactor}
  MaterialExpressionPixelDepth_1 = PixelDepth()
  MaterialExpressionScalarParameter_32 = ScalarParameter() {group=ContactShadow, parameterName=ContactShadow_JitterOffset}
  n38 = Custom!(
   WorldPosition=@MaterialExpressionWorldPosition_4.XYZ
   DepthTexture=@MaterialExpressionTextureObjectParameter_4
   LightDirection=@MaterialExpressionSkyAtmosphereLightDirection_3
   RayMarchingDistance=@MaterialExpressionScalarParameter_8
   BaseUV=@MaterialExpressionTextureCoordinate_2
   CameraPosition=@MaterialExpressionCameraPositionWS_5
   WorldPositionBehindTranslucent=@MaterialExpressionMaterialFunctionCall_13.Result
   XCameraToWorldVector=@MaterialExpressionTransform_3
   YCameraToWorldVector=@MaterialExpressionTransform_4
   ZCameraToWorldVector=@MaterialExpressionTransform_5
   HalfFOVAngle=@MaterialExpressionViewProperty_0.Property
   ScreenUV=@MaterialExpressionScreenPosition_0.ViewportUV
   MaxDistance=@MaterialExpressionScalarParameter_13
   WTU_A=@MaterialExpressionAppendVector_0
   WTU_B=@MaterialExpressionAppendVector_1
   WTU_C=@MaterialExpressionAppendVector_2
   WTU_D=@MaterialExpressionAppendVector_3
   CameraDirection=@MaterialExpressionMaterialFunctionCall_19.Result
   RandomTex=@MaterialExpressionNamedRerouteUsage_1
   LoopCount=@MaterialExpressionLinearInterpolate_10
   StepDistanceJitterFactor=@MaterialExpressionScalarParameter_19
   LightDirectionJitterAngle=@MaterialExpressionScalarParameter_20
   DepthFixOffsetFactor=@MaterialExpressionScalarParameter_23
   PixelDepth=@MaterialExpressionPixelDepth_1.R
   JitterOffset=@MaterialExpressionScalarParameter_32
  ) {
   code="\r\nfloat4x4 WorldToUnit = {WTU_A,WTU_B,WTU_C,WTU_D};\r\nfloat ShadowValid = 0;\r\nfloat StepDistance = RayMarchingDistance/LoopCount;\r\n//===TempJitter===\r\nfloat2 ScreenUV2 = MaterialFloat2(ScreenAlignedPosition(Parameters.ScreenPosition).xy);\r\nfloat2 ScenePixels=View.BufferSizeAndInvSize.xy*ScreenUV2;\r\nScenePixels+=View.TemporalAAParams.r;\r\nfloat2 RandomSamp = ((uint)(ScenePixels.x) + 2 * (uint)(ScenePixels.y)) % 5;\r\nRandomSamp+=Texture2DSample(RandomTex,RandomTexSampler,ScenePixels);\r\nRandomSamp/=5;//5\r\nRandomSamp-=JitterOffset;\r\n//===TempJitter===\r\nfloat TempAADistance = 1;\r\nTempAADistance*=StepDistance*lerp(RandomSamp.x,1.0,StepDistanceJitterFactor);\r\n//Use random direction>>\r\nfloat3 axis = LightDirection;\r\nfloat coneAngle = LightDirectionJitterAngle;\r\n//===Random vector in cone===\r\n// find a random point on portal coord\r\nfloat theta = 2.0 * 3.14159265 * saturate(RandomSamp.x); // azimuthal angle\r\nfloat r = sqrt(saturate(RandomSamp.x)*0.815); // radius\r\nfloat x = r * cos(theta);\r\nfloat y = r * sin(theta);\r\n// Create ref \"x\"\r\nfloat3 referenceVec = abs(axis.x) > abs(axis.y) ? float3(-axis.z, 0, axis.x) : float3(0, axis.z, -axis.y);\r\nreferenceVec = normalize(referenceVec);\r\n// Create ref \"y\"\r\nfloat3 orthogonalVec = cross(axis, referenceVec);\r\northogonalVec = normalize(orthogonalVec);\r\n// apply random point\r\nfloat3 randomPoint = x * referenceVec + y * orthogonalVec;\r\n// find final \"z\"\r\nfloat cosAngle = cos(coneAngle);\r\nfloat z = lerp(1, cosAngle, saturate(RandomSamp.x)*0.815);\r\nfloat3 randomVec = normalize(z * axis + sqrt(1 - z * z) * randomPoint);\r\n//===Random vector in cone===\r\nfor (int i = 1;i<LoopCount;i++){\r\n\r\n    float3 RayMarchingDirection = TempAADistance*i*lerp(LightDirection,randomVec,1);\r\n    float3 RayMarchingPosition = WorldPositionBehindTranslucent+RayMarchingDirection;\r\n    float3 ProjectedRayMarchingDirection=(RayMarchingDirection-dot(RayMarchingDirection,CameraDirection)*CameraDirection)*(WorldPosition-CameraPosition)/(WorldPositionBehindTranslucent-CameraPosition);\r\n    \r\n    float4 TranslatedRayMarchPos = mul(float4(WorldPosition+ProjectedRayMarchingDirection,1.0),WorldToUnit);\r\n    float2 NewUV = TranslatedRayMarchPos.xy;\r\n    float SampledDepth = Texture2DSampleLevel(DepthTexture,DepthTextureSampler,NewUV,0.0).a;\r\n    \r\n    float3 CameraRelativePosition = RayMarchingPosition-CameraPosition;\r\n    float RayMarchingPosDepth = length(CameraRelativePosition);\r\n    float DepthFixOffset = (distance(WorldPosition,CameraPosition)-PixelDepth)*DepthFixOffsetFactor;\r\n    float DepthDistance = (RayMarchingPosDepth - (SampledDepth+DepthFixOffset));\r\n\r\n    ShadowValid = saturate(ceil(DepthDistance));\r\n    ShadowValid *= saturate((MaxDistance-DepthDistance)) * (1-(float)i /LoopCount);\r\n    if(ShadowValid>0)\r\n    {\r\n        break; \r\n    }\r\n}\r\nreturn ShadowValid;"
   inputs [
    {
     inputName=WorldPosition
     input {expression=&self:MaterialExpressionWorldPosition_4, mask=1, maskB=1, maskG=1, maskR=1}
    }
    {inputName=DepthTexture, input {expression=&self:MaterialExpressionTextureObjectParameter_4}}
    {
     inputName=LightDirection
     input {expression=&self:MaterialExpressionSkyAtmosphereLightDirection_3}
    }
    {inputName=RayMarchingDistance, input {expression=&self:MaterialExpressionScalarParameter_8}}
    {inputName=BaseUV, input {expression=&self:MaterialExpressionTextureCoordinate_2}}
    {inputName=CameraPosition, input {expression=&self:MaterialExpressionCameraPositionWS_5}}
    {
     inputName=WorldPositionBehindTranslucent
     input {expression=&self:MaterialExpressionMaterialFunctionCall_13}
    }
    {inputName=XCameraToWorldVector, input {expression=&self:MaterialExpressionTransform_3}}
    {inputName=YCameraToWorldVector, input {expression=&self:MaterialExpressionTransform_4}}
    {inputName=ZCameraToWorldVector, input {expression=&self:MaterialExpressionTransform_5}}
    {inputName=HalfFOVAngle, input {expression=&self:MaterialExpressionViewProperty_0}}
    {inputName=ScreenUV, input {expression=&self:MaterialExpressionScreenPosition_0}}
    {inputName=MaxDistance, input {expression=&self:MaterialExpressionScalarParameter_13}}
    {inputName=WTU_A, input {expression=&self:MaterialExpressionAppendVector_0}}
    {inputName=WTU_B, input {expression=&self:MaterialExpressionAppendVector_1}}
    {inputName=WTU_C, input {expression=&self:MaterialExpressionAppendVector_2}}
    {inputName=WTU_D, input {expression=&self:MaterialExpressionAppendVector_3}}
    {
     inputName=CameraDirection
     input {expression=&self:MaterialExpressionMaterialFunctionCall_19}
    }
    {inputName=RandomTex, input {expression=&self:MaterialExpressionNamedRerouteUsage_1}}
    {inputName=LoopCount, input {expression=&self:MaterialExpressionLinearInterpolate_10}}
    {
     inputName=StepDistanceJitterFactor
     input {expression=&self:MaterialExpressionScalarParameter_19}
    }
    {
     inputName=LightDirectionJitterAngle
     input {expression=&self:MaterialExpressionScalarParameter_20}
    }
    {
     inputName=DepthFixOffsetFactor
     input {expression=&self:MaterialExpressionScalarParameter_23}
    }
    {
     inputName=PixelDepth
     input {expression=&self:MaterialExpressionPixelDepth_1, mask=1, maskR=1}
    }
    {inputName=JitterOffset, input {expression=&self:MaterialExpressionScalarParameter_32}}
   ]
  }
  n39 = NamedRerouteUsage()
  n40 = Power(Base=Saturate(None=Subtract(A=@n38, B=@n39))) {constExponent=1.2}
  n41 = Multiply(
   A=Multiply(
    A=ScalarParameter() {defaultValue=0.1, group=ContactShadow, parameterName=ContactShadow_Strength}
    B=ScalarParameter() {defaultValue=1, group=FromNiagara, parameterName=ShadowStrengthScale}
   )
   B=ScalarParameter() {defaultValue=1, group=Opacity, parameterName=TotalOpacityScale}
  )
  NamedRerouteDeclaration(None=StaticSwitchParameter(True=Constant(), False=Multiply(A=@n40, B=@n41) {constB=0.75}) {defaultValue=true, parameterName=DisableContactShadow}) {name=ContactShadow, nodeColor=<rgba 0 0.7656 1 1>}
  n42 = ScalarParameter() {defaultValue=1, group=Base, parameterName=ScaleBlurDistance}
  n43 = MaterialFunctionCall(Float2=ParticleSize()) {materialFunction=&BreakOutFloat2Components}
  n44 = Max(A=Divide(B=@n43.R) {constA=300, constB=300}, B=Divide(B=@n43.G) {constA=300, constB=300})
  NamedRerouteDeclaration(None=Multiply(A=Multiply(A=@n42, B=@n44), B=ScalarParameter() {defaultValue=1, group=FromNiagara, parameterName=NiagaraScaleBlurDistance})) {name=SmoothDistanceScale, nodeColor=<rgba 0.7578 0 1 1>}
  NamedRerouteDeclaration(None=VectorParameter() {group=FakeSSS, parameterName=FakeSSS_Color, defaultValue=<rgba 0 0 0 1>}.RGB) {name=SSSColor, nodeColor=<rgba 0 1 0.0391 1>}
  n45 = DotProduct(A=CameraVectorWS(), B=SkyAtmosphereLightDirection())
  n46 = Power(Base=Saturate(None=Multiply(A=@n45) {constB=-1}), Exp=ScalarParameter() {defaultValue=5, group=FakeSSS, parameterName=FakeSSS_AnglePower})
  n47 = NamedRerouteUsage()
  n48 = VectorParameter() {group=FakeSSS, parameterName=FakeSSS_RemapThickness, channelNames {a="", b="", g=TargetHigh, r=TargetLow}, defaultValue=<rgba 0.2 1 0 1>}
  n49 = MaterialFunctionCall(
   Input=Power(Base=OneMinus(None=@n47), Exp=ScalarParameter() {defaultValue=1, group=FakeSSS, parameterName=FakeSSS_ThicknessPower})
   Input Low=Constant()
   Input High=Constant() {r=1}
   Target Low=@n48.TargetLow
   Target High=@n48.TargetLow
  ) {materialFunction=&RemapValueRange}
  NamedRerouteDeclaration(None=Saturate(None=LinearInterpolate(B=@n46, Alpha=@n49.Result) {constB=0})) {name=SSSStrength, nodeColor=<rgba 0 1 0.4141 1>}
  NamedRerouteDeclaration(None=@n3) {name=OriginThicknessOutput, nodeColor=<rgba 1 0.4576 0 1>}
  MaterialExpressionTextureObjectParameter_0 = TextureObjectParameter() {group=FromNiagara, parameterName=Tex}
  MaterialExpressionMultiply_3 = Multiply(A=ScalarParameter() {defaultValue=0.012, group=NormalDepth, parameterName=NormalDepth_BlurDistance}, B=NamedRerouteUsage())
  MaterialExpressionScalarParameter_4 = ScalarParameter() {defaultValue=8, group=NormalDepth, parameterName=NormalDepth_DistanceSteps}
  MaterialExpressionScalarParameter_5 = ScalarParameter() {defaultValue=6, group=NormalDepth, parameterName=NormalDepth_RadialSteps}
  MaterialExpressionConstant_5 = Constant() {r=0.618}
  MaterialExpressionScalarParameter_9 = ScalarParameter() {defaultValue=0.1, group=NormalDepth, parameterName=NormalDepth_KernelPower}
  MaterialExpressionNamedRerouteUsage_26 = NamedRerouteUsage()
  MaterialExpressionConstant_12 = Constant()
  MaterialExpressionScalarParameter_25 = ScalarParameter() {defaultValue=0.5, group=NormalDepth, parameterName=NormalDepth_JitterIntensity}
  MaterialExpressionConstant_10 = Constant()
  n50 = Custom!(
   Tex=@MaterialExpressionTextureObjectParameter_0
   UV=@MaterialExpressionStaticSwitchParameter_12
   Distance=@MaterialExpressionMultiply_3
   DistanceSteps=@MaterialExpressionScalarParameter_4
   RadialSteps=@MaterialExpressionScalarParameter_5
   RadialOffset=@MaterialExpressionConstant_5
   KernelPower=@MaterialExpressionScalarParameter_9
   RandomTex=@MaterialExpressionNamedRerouteUsage_26
   JitterRandomIntensity=@MaterialExpressionConstant_12
   JitterIntensity=@MaterialExpressionScalarParameter_25
   ClampSmpResult=@MaterialExpressionConstant_10
  ) {
   code="\r\nfloat4 CurColor=0;\r\nfloat2 NewUV = UV;\r\nint i=0;\r\nfloat StepSize = Distance / (int) DistanceSteps;\r\nfloat CurDistance=0;\r\nfloat2 CurOffset=0;\r\nfloat SubOffset = 0;\r\nfloat TwoPi = 6.283185;\r\nfloat accumdist=0;\r\nfloat2 ScreenUV = MaterialFloat2(ScreenAlignedPosition(Parameters.ScreenPosition).xy);\r\nfloat2 ScenePixels=View.BufferSizeAndInvSize.xy*ScreenUV;\r\nScenePixels+=View.TemporalAAParams.r;\r\nfloat2 RandomSamp = ((uint)(ScenePixels.x) + 2 * (uint)(ScenePixels.y)) % 5;\r\n//RandomSamp+=Texture2DSample(RandomTex,RandomTexSampler,ScenePixels)*JitterRandomIntensity;\r\nRandomSamp/=5;\r\nRandomSamp-=0.5;\r\nRandomSamp*=JitterIntensity;\r\nfloat TempAARotation = 1;\r\nfloat TempAADistance = 1;\r\nTempAARotation*=RandomSamp.x;\r\nTempAADistance*=StepSize*RandomSamp.x;\r\n\r\nif (DistanceSteps < 1)\r\n\r\n{\r\nreturn Texture2DSample(Tex,TexSampler,UV);\t\t\r\n}\r\nelse\r\n{\r\nwhile ( i < (int) DistanceSteps)\r\n{\r\n\tCurDistance += StepSize;\r\n\tfor (int j = 0; j < (int) RadialSteps; j++)\r\n\t{\r\n\t\tSubOffset +=1;\r\n\t\tCurOffset.x = cos(TwoPi*((TempAARotation+SubOffset) / RadialSteps));\r\n\t\tCurOffset.y = sin(TwoPi*((TempAARotation+SubOffset) / RadialSteps));\r\n\t\tNewUV.x = UV.x + CurOffset.x * (CurDistance+(RandomSamp*TempAADistance));\r\n\t\tNewUV.y = UV.y + CurOffset.y * (CurDistance+(RandomSamp*TempAADistance));\r\n\t\tfloat distpow = pow(CurDistance, KernelPower);\r\n\t\tfloat4 tmpTexSmp = Texture2DSample(Tex,TexSampler,NewUV);\r\n\t\tCurColor += lerp(tmpTexSmp,clamp(tmpTexSmp,0,100000),ClampSmpResult)*distpow;\t\t\r\n\t\taccumdist += distpow;\r\n\t}\r\nSubOffset +=RadialOffset;\r\ni++;\r\n}\r\nCurColor = CurColor;\r\nCurColor /=accumdist;\r\nreturn CurColor;\r\n}\r\n\r\n"
   outputType=CMOT_Float4
   inputs [
    {inputName=Tex, input {expression=&self:MaterialExpressionTextureObjectParameter_0}}
    {inputName=UV, input {expression=&self:MaterialExpressionStaticSwitchParameter_12}}
    {inputName=Distance, input {expression=&self:MaterialExpressionMultiply_3}}
    {inputName=DistanceSteps, input {expression=&self:MaterialExpressionScalarParameter_4}}
    {inputName=RadialSteps, input {expression=&self:MaterialExpressionScalarParameter_5}}
    {inputName=RadialOffset, input {expression=&self:MaterialExpressionConstant_5}}
    {inputName=KernelPower, input {expression=&self:MaterialExpressionScalarParameter_9}}
    {inputName=RandomTex, input {expression=&self:MaterialExpressionNamedRerouteUsage_26}}
    {inputName=JitterRandomIntensity, input {expression=&self:MaterialExpressionConstant_12}}
    {inputName=JitterIntensity, input {expression=&self:MaterialExpressionScalarParameter_25}}
    {inputName=ClampSmpResult, input {expression=&self:MaterialExpressionConstant_10}}
   ]
  }
  NamedRerouteDeclaration(None=ComponentMask(None=@n50) {b=true, g=true, r=true}) {name=FluidNormal, nodeColor=<rgba 0.2187 1 0 1>}
  NamedRerouteDeclaration(None=ComponentMask(None=@n50) {a=true}) {name=FluidDepth, nodeColor=<rgba 0 0.1797 1 1>}
  n51 = Subtract(
   A=Divide(A=NamedRerouteUsage(), B=ScalarParameter() {defaultValue=40960, group=Opacity, parameterName=DepthToOpacity_Divide})
   B=Add(A=ScalarParameter() {group=Opacity, parameterName=EdgeOffset}, B=ScalarParameter() {group=FromNiagara, parameterName=NiagaraEdgeOffset})
  )
  n52 = MaterialFunctionCall(
   In=Power(Base=Saturate(None=@n51), Exp=ScalarParameter() {defaultValue=5, group=Opacity, parameterName=DepthToOpacity_Power})
   Contrast=ScalarParameter() {defaultValue=2, group=Opacity, parameterName=DepthToOpacity_Contrast}
  ) {materialFunction=&CheapContrast}
  NamedRerouteDeclaration(None=OneMinus(None=Saturate(None=@n52.Result))) {name=FluidOpacity, nodeColor=<rgba 0.5703 0 1 1>}
  n53 = ScalarParameter() {defaultValue=10, group=Refraction, parameterName=Refraction_StrengthA}
  n54 = ScalarParameter() {defaultValue=80, group=Refraction, parameterName=Refraction_StrengthB}
  n55 = NamedRerouteUsage()
  n56 = CameraVectorWS()
  n57 = Multiply(
   A=Multiply(A=LinearInterpolate(A=@n53, B=@n54, Alpha=@n55), B=ScalarParameter() {defaultValue=1, group=FromNiagara, parameterName=ScaleRefraction})
   B=MaterialFunctionCall(
    Ray Direction=Multiply(A=@n56)
    Surface Normal=NamedRerouteUsage()
    Refractive Index Target=ScalarParameter() {defaultValue=1.03, group=Refraction, parameterName=Refraction_IndexTarget}
   ) {materialFunction=&Refract}.Vector
  )
  n58 = WorldPosition()
  n59 = MaterialFunctionCall(Camera Relative Position=Subtract(A=Add(A=@n57, B=@n58.XYZ), B=CameraPositionWS())) {materialFunction=&ManualWorldToScreenUVsTransform}
  NamedRerouteDeclaration(None=StaticSwitchParameter(True=Constant() {r=1}, False=SceneColor!(UVs=@n59.Screen UVs).RGB) {parameterName=DisableSceneColorSample}) {name=SampledSceneColor, nodeColor=<rgba 0 1 0.9297 1>}
  n60 = TextureSampleParameterCube!(UVs=ReflectionVectorWS(CustomWorldNormal=NamedRerouteUsage()), Apply View MipBias) {
   group=HDRIReflection
   parameterName=HDRIReflectionTexture
   samplerType=SAMPLERTYPE_LinearColor
   texture=&/T_WaterReflection
  }
  n61 = Multiply(A=@n60.RGB, B=VectorParameter() {group=HDRIReflection, parameterName=HDRIReflectionTint, defaultValue=<rgba 1 1 1 1>}.RGB)
  n62 = VectorParameter() {group=HDRIReflection, parameterName=HDRIReflectionFresnel, channelNames {a="", b="", g=Base, r=Exp}, defaultValue=<rgba 2 0.2 0 1>}
  n63 = Fresnel(ExponentIn=@n62.Exp, BaseReflectFractionIn=@n62.Exp, Normal=NamedRerouteUsage())
  NamedRerouteDeclaration(None=StaticSwitchParameter(True=Constant(), False=Multiply(A=@n61, B=@n63)) {parameterName=DisableHDRIReflection}) {name=HDRIReflection, nodeColor=<rgba 1 0 0.0469 1>}
  n64 = ReflectionVectorWS(CustomWorldNormal=NamedRerouteUsage())
  n65 = Multiply(
   A=VectorParameter() {group=EnvReflection, parameterName=EnvReflectionTint, defaultValue=<rgba 0.125 0.125 0.125 1>}.RGB
   B=Desaturation(None=SkyLightEnvMapSample(Direction=@n64), Fraction=ScalarParameter() {group=EnvReflection, parameterName=EnvReflectionDesaturation})
  )
  n66 = VectorParameter() {group=EnvReflection, parameterName=EnvReflectionFresnel, channelNames {a="", b="", g=Base, r=Exp}, defaultValue=<rgba 2 0 0 1>}
  n67 = Fresnel(ExponentIn=@n66.Exp, BaseReflectFractionIn=@n66.Exp, Normal=NamedRerouteUsage())
  NamedRerouteDeclaration(None=StaticSwitchParameter(True=Constant(), False=LinearInterpolate(B=@n65, Alpha=@n67)) {parameterName=DisableEnvReflection}) {name=EnvReflection, nodeColor=<rgba 0 1 0.6484 1>}
  n68 = VectorParameter() {group=Lighting, parameterName=LightingAttribute, channelNames {a="", b="", g=Contrast, r=Offset}, defaultValue=<rgba 0.6 0.5 0 0>}
  n69 = Add(A=DotProduct(A=SkyAtmosphereLightDirection(), B=NamedRerouteUsage()), B=@n68.Offset)
  NamedRerouteDeclaration(None=MaterialFunctionCall(In=Saturate(None=@n69), Contrast=@n68.Contrast) {materialFunction=&CheapContrast}.Result) {name=LightingResult, nodeColor=<rgba 0 0.4141 1 1>}
  n70 = VectorParameter() {
   group=EmissiveColor
   parameterName=ThicknessEmissiveLighting
   channelNames {a="", b="", g=LightFace, r=DarkFace}
   defaultValue=<rgba 1 1 0 0>
  }
  n71 = NamedRerouteUsage()
  n72 = VectorParameter() {group=EmissiveColor, parameterName=ThicknessEmissiveColor_LowThickness, defaultValue=<rgba 0 0 0 1>}
  n73 = VectorParameter() {group=EmissiveColor, parameterName=ThicknessEmissiveColor_HighThickness, defaultValue=<rgba 0 0 0 1>}
  n74 = NamedRerouteUsage()
  NamedRerouteDeclaration(None=Multiply(A=LinearInterpolate(A=@n70.DarkFace, B=@n70.DarkFace, Alpha=@n71), B=LinearInterpolate(A=@n72.RGB, B=@n73.RGB, Alpha=@n74))) {name=ThicknessEmissive, nodeColor=<rgba 1 0 0.5625 1>}
  n75 = NamedRerouteUsage()
  n76 = OneMinus(None=NamedRerouteUsage())
  n77 = ScalarParameter() {defaultValue=1, group=Opacity, parameterName=ThicknessToOpacityThicknessPower}
  n78 = VectorParameter() {group=Opacity, parameterName=ThicknessToOpacity_Remap, channelNames {a="", b="", g=TargetHigh, r=TargetLow}, defaultValue=<rgba 0.5 1 0 1>}
  n79 = MaterialFunctionCall(
   Input=OneMinus(None=Power(Base=@n76, Exp=@n77))
   Input Low=Constant()
   Input High=Constant() {r=1}
   Target Low=@n78.TargetLow
   Target High=@n78.TargetLow
  ) {materialFunction=&RemapValueRange}
  n80 = Multiply(
   A=ScalarParameter() {defaultValue=1, group=Opacity, parameterName=TotalOpacityScale}
   B=StaticSwitchParameter(True=Saturate(None=@n79.Result), False=Constant() {r=1}) {parameterName=EnableThicknessToOpacity}
  )
  n81 = Multiply(A=@n75, B=@n80)
  n82 = Add(A=NamedRerouteUsage(), B=@n75)
  n83 = StaticSwitchParameter(True=Multiply(A=@n82, B=@n80), False=@n82) {parameterName=DisableSceneColorSample}
  n84 = MaterialFunctionCall(Float2=TextureCoordinate()) {materialFunction=&BreakOutFloat2Components}
  n85 = VectorParameter() {group=EdgeMask, parameterName=EdgeMask_Control, channelNames {a="", b=Power, g=Multiply, r=Offset}, defaultValue=<rgba 0 5 2 1>}
  n86 = Multiply(
   A=MaterialFunctionCall!(UorV=@n84.R, EdgeControl=@n85) {materialFunction=/Game/Effects/Niagara/MasterMaterials/Functions/Common/FXCommon_EdgeMask_Simple}.Mask
   B=MaterialFunctionCall(UorV=@n84.G, EdgeControl=@n85) {materialFunction=&FXCommon_EdgeMask_Simple}.Mask
  )
  n87 = Constant()
  ShadingPathSwitch(
   Default=StaticSwitchParameter(True=Ceil(None=@n81), False=Multiply(A=@n83, B=@n86)) {parameterName=BakeNormalAndThickness}
   Forward=@n87
   Mobile=@n87
  )
  n88 = VectorParameter() {group=ContactShadow, parameterName=ContactShadow_Color, defaultValue=<rgba 0 0 0 1>}
  n89 = StaticSwitchParameter(True=Constant() {r=1}, False=@n80) {parameterName=DisableSceneColorSample}
  n90 = LinearInterpolate(B=NamedRerouteUsage(), Alpha=@n89) {constAlpha=0}
  n91 = Ceil(None=NamedRerouteUsage())
  ShadingPathSwitch(
   Default=StaticSwitchParameter(True=Constant(), False=LinearInterpolate(A=@n88.RGB, B=@n90, Alpha=@n91)) {parameterName=BakeNormalAndThickness}
   Forward=@n87
   Mobile=@n87
  )
  MaterialExpressionTextureCoordinate_6 = TextureCoordinate() {uTiling=2, vTiling=2}
  n92 = MaterialFunctionCall(Float2=@MaterialExpressionTextureCoordinate_6) {materialFunction=&BreakOutFloat2Components}
  n93 = Constant()
  n94 = Constant() {r=1}
  n95 = If(A=@n92.R, "A > B"=@n93, "A == B"=@n94, "A < B"=@n94) {constB=1}
  n96 = If(A=@n92.G, "A > B"=@n94, "A == B"=@n93, "A < B"=@n93) {constB=1}
  n97 = Multiply(A=@n95, B=@n96)
  n98 = If(A=@n92.R, "A > B"=@n94, "A == B"=@n93, "A < B"=@n93) {constB=1}
  n99 = If(A=@n92.G, "A > B"=@n93, "A == B"=@n94, "A < B"=@n94) {constB=1}
  n100 = Multiply(A=@n98, B=@n99)
  n101 = NamedRerouteUsage()
  n102 = Transform(None=NamedRerouteUsage()) {transformSourceType=TRANSFORMSOURCE_World, transformType=TRANSFORM_Tangent}
  n103 = MaterialFunctionCall(Float3=Multiply(A=Add(A=@n102)) {constB=0.5}) {materialFunction=&BreakOutFloat3Components}
  n104 = Add(
   A=Add(A=Multiply(A=@n97, B=@n81), B=Multiply(A=@n100, B=@n101))
   B=Multiply(
    A=Multiply(A=@n95, B=@n99)
    B=MaterialFunctionCall(X=@n103.R, Y=@n103.R, Z=@n103.R) {materialFunction=&MakeFloat3}.Result
   )
  )
  MaterialExpressionTextureObjectParameter_3 = TextureObjectParameter() {group=FromNiagara, parameterName=TexVelocity}
  MaterialExpressionMultiply_40 = Multiply(A=ScalarParameter() {defaultValue=0.1, group=BakeMotionVector, parameterName=BakeMotionVector_BlurDistance}, B=NamedRerouteUsage())
  MaterialExpressionScalarParameter_44 = ScalarParameter() {defaultValue=35, group=BakeMotionVector, parameterName=BakeMotionVector_DistanceSteps}
  MaterialExpressionScalarParameter_52 = ScalarParameter() {defaultValue=12, group=BakeMotionVector, parameterName=BakeMotionVector_RadialSteps}
  MaterialExpressionConstant_17 = Constant() {r=0.618}
  MaterialExpressionScalarParameter_47 = ScalarParameter() {defaultValue=1, group=BakeMotionVector, parameterName=BakeMotionVector_KernelPower}
  MaterialExpressionNamedRerouteUsage_40 = NamedRerouteUsage()
  MaterialExpressionConstant_27 = Constant()
  MaterialExpressionScalarParameter_48 = ScalarParameter() {defaultValue=0.5, group=BakeMotionVector, parameterName=BakeMotionVector_JitterIntensity}
  MaterialExpressionConstant_26 = Constant()
  n105 = Custom(
   Tex=@MaterialExpressionTextureObjectParameter_3
   UV=@MaterialExpressionTextureCoordinate_6
   Distance=@MaterialExpressionMultiply_40
   DistanceSteps=@MaterialExpressionScalarParameter_44
   RadialSteps=@MaterialExpressionScalarParameter_52
   RadialOffset=@MaterialExpressionConstant_17
   KernelPower=@MaterialExpressionScalarParameter_47
   RandomTex=@MaterialExpressionNamedRerouteUsage_40
   JitterRandomIntensity=@MaterialExpressionConstant_27
   JitterIntensity=@MaterialExpressionScalarParameter_48
   ClampSmpResult=@MaterialExpressionConstant_26
  ) {
   code="\r\nfloat4 CurColor=0;\r\nfloat2 NewUV = UV;\r\nint i=0;\r\nfloat StepSize = Distance / (int) DistanceSteps;\r\nfloat CurDistance=0;\r\nfloat2 CurOffset=0;\r\nfloat SubOffset = 0;\r\nfloat TwoPi = 6.283185;\r\nfloat accumdist=0;\r\nfloat2 ScreenUV = MaterialFloat2(ScreenAlignedPosition(Parameters.ScreenPosition).xy);\r\nfloat2 ScenePixels=View.BufferSizeAndInvSize.xy*ScreenUV;\r\nScenePixels+=View.TemporalAAParams.r;\r\nfloat2 RandomSamp = ((uint)(ScenePixels.x) + 2 * (uint)(ScenePixels.y)) % 5;\r\n//RandomSamp+=Texture2DSample(RandomTex,RandomTexSampler,ScenePixels)*JitterRandomIntensity;\r\nRandomSamp/=5;\r\nRandomSamp-=0.5;\r\nRandomSamp*=JitterIntensity;\r\n//RandomSamp+=1;\r\nfloat TempAARotation = 1;\r\nfloat TempAADistance = 1;\r\nTempAARotation*=RandomSamp.x;\r\nTempAADistance*=StepSize*RandomSamp.x;\r\n\r\nif (DistanceSteps < 1)\r\n\r\n{\r\nreturn Texture2DSample(Tex,TexSampler,UV);\t\t\r\n}\r\nelse\r\n{\r\nwhile ( i < (int) DistanceSteps)\r\n{\r\n\tCurDistance += StepSize;\r\n\tfor (int j = 0; j < (int) RadialSteps; j++)\r\n\t{\r\n\t\tSubOffset +=1;\r\n\t\tCurOffset.x = cos(TwoPi*((TempAARotation+SubOffset) / RadialSteps));\r\n\t\tCurOffset.y = sin(TwoPi*((TempAARotation+SubOffset) / RadialSteps));\r\n\t\tNewUV.x = UV.x + CurOffset.x * (CurDistance+(RandomSamp*TempAADistance));\r\n\t\tNewUV.y = UV.y + CurOffset.y * (CurDistance+(RandomSamp*TempAADistance));\r\n\t\tfloat distpow = pow(CurDistance, KernelPower);\r\n\t\tfloat4 tmpTexSmp = Texture2DSample(Tex,TexSampler,NewUV);\r\n\t\tCurColor += lerp(tmpTexSmp,clamp(tmpTexSmp,0,100000),ClampSmpResult)*distpow;\t\t\r\n\t\taccumdist += distpow;\r\n\t}\r\nSubOffset +=RadialOffset;\r\ni++;\r\n}\r\nCurColor = CurColor;\r\nCurColor /=accumdist;\r\nreturn CurColor;\r\n}\r\n\r\n"
   outputType=CMOT_Float4
   inputs [
    {inputName=Tex, input {expression=&self:MaterialExpressionTextureObjectParameter_3}}
    {inputName=UV, input {expression=&self:MaterialExpressionTextureCoordinate_6}}
    {inputName=Distance, input {expression=&self:MaterialExpressionMultiply_40}}
    {inputName=DistanceSteps, input {expression=&self:MaterialExpressionScalarParameter_44}}
    {inputName=RadialSteps, input {expression=&self:MaterialExpressionScalarParameter_52}}
    {inputName=RadialOffset, input {expression=&self:MaterialExpressionConstant_17}}
    {inputName=KernelPower, input {expression=&self:MaterialExpressionScalarParameter_47}}
    {inputName=RandomTex, input {expression=&self:MaterialExpressionNamedRerouteUsage_40}}
    {inputName=JitterRandomIntensity, input {expression=&self:MaterialExpressionConstant_27}}
    {inputName=JitterIntensity, input {expression=&self:MaterialExpressionScalarParameter_48}}
    {inputName=ClampSmpResult, input {expression=&self:MaterialExpressionConstant_26}}
   ]
  }
  n106 = Divide(A=ComponentMask(None=@n105) {g=true, r=true}) {constB=512}
  n107 = Multiply(A=Add(A=Multiply(A=@n106) {constB=-1})) {constB=0.5}
  n108 = Multiply(A=Multiply(A=@n96, B=@n98), B=AppendVector(A=Saturate(None=@n107), B=Constant() {r=1}))
  n109 = TextureCoordinate()
  n110 = Constant2Vector() {g=1400, r=1400}
  n111 = ParticleSize()
  n112 = MaterialFunctionCall!(OffsetAmount=Multiply(A=Divide(A=@n110, B=@n111) {constA=1400, constB=1400}, B=Constant() {r=0.01})) {materialFunction=/Engine/Functions/Engine_MaterialFunctions01/ImageAdjustment/BlurSampleOffsets}
  n113 = Add(A=@n109, B=@n112."-.5,.5")
  n114 = TextureObjectParameter() {group=FromNiagara, parameterName=Tex}
  n115 = Add(A=@n109, B=@n112.".5,.5")
  n116 = Min(
   A=Min(A=TextureSample(UVs=@n113, Tex=@n114).A, B=TextureSample(UVs=@n115, Tex=@n114).A)
   B=TextureSample(UVs=Add(A=@n109, B=@n112.".5,-.5"), Tex=@n114).A
  )
  n117 = TextureSample(UVs=Add(A=@n109, B=@n112."-.5,-.5"), Tex=@n114)
  MaterialExpressionLinearInterpolate_11 = LinearInterpolate(A=Min(A=Min(A=@n116, B=@n117.A), B=TextureSample(Tex=@n114).A), Alpha=ScalarParameter() {parameterName=DebugBound}) {constB=0}
  n118 = Custom!(Opacity=@MaterialExpressionLinearInterpolate_11) {
   code="float mask = 0;\r\nif (Opacity >= 65000)\r\n{\r\nclip(-1);\r\n}\r\nreturn mask;"
   outputType=CMOT_Float1
   inputs [{inputName=Opacity, input {expression=&self:MaterialExpressionLinearInterpolate_11}}]
  }
  n119 = Add(A=NamedRerouteUsage(), B=NamedRerouteUsage())
  n120 = MaterialFunctionCall(Emissive=@n119) {materialFunction=&MF_FX_EmissiveAdjustments}
  n121 = NamedRerouteUsage()
  n122 = LinearInterpolate(A=@n121, B=Add(A=NamedRerouteUsage(), B=@n121), Alpha=NamedRerouteUsage())
  n123 = LinearInterpolate(
   A=NamedRerouteUsage()
   B=Add(
    A=MaterialFunctionCall(Emissive=@n122) {materialFunction=&MF_FX_EmissiveAdjustments}.Result
    B=NamedRerouteUsage()
   )
   Alpha=@n89
  ) {constAlpha=0}
  n124 = NamedRerouteUsage()
  n125 = Add(
   A=MaterialFunctionCall(Emissive=Add(A=@n119, B=@n122)) {materialFunction=&MF_FX_EmissiveAdjustments}.Result
   B=NamedRerouteUsage()
  )
  n126 = LinearInterpolate(
   A=@n88.RGB
   B=StaticSwitchParameter(True=Add(A=@n120.Result, B=@n123), False=LinearInterpolate(A=@n124, B=@n125, Alpha=@n89) {constAlpha=0}) {parameterName=ThicknessToOpacityNotEffectReflection}
   Alpha=@n91
  )
  ShadingPathSwitch(
   Default=StaticSwitchParameter(True=Add(A=@n104, B=@n108), False=Add(A=@n118, B=@n126)) {parameterName=BakeNormalAndThickness}
   Forward=@n87
   Mobile=@n87
  )
  MaterialExpressionSkyAtmosphereLightDirection_2 = SkyAtmosphereLightDirection()
  MaterialExpressionMaterialFunctionCall_0 = MaterialFunctionCall() {materialFunction=&WorldPositionBehindTranslucency}
  MaterialExpressionMaterialFunctionCall_11 = MaterialFunctionCall() {materialFunction=&CameraDirectionVector}
  MaterialExpressionCameraPositionWS_1 = CameraPositionWS()
  n127 = VectorParameter() {group=FromNiagara, parameterName=WorldToUnit_Row0}
  MaterialExpressionAppendVector_4 = AppendVector(A=@n127.RGB, B=@n127.RGB)
  n128 = VectorParameter() {group=FromNiagara, parameterName=WorldToUnit_Row1}
  MaterialExpressionAppendVector_5 = AppendVector(A=@n128.RGB, B=@n128.RGB)
  n129 = VectorParameter() {group=FromNiagara, parameterName=WorldToUnit_Row2}
  MaterialExpressionAppendVector_6 = AppendVector(A=@n129.RGB, B=@n129.RGB)
  n130 = VectorParameter() {group=FromNiagara, parameterName=WorldToUnit_Row3}
  MaterialExpressionAppendVector_7 = AppendVector(A=@n130.RGB, B=@n130.RGB)
  MaterialExpressionWorldPosition_3 = WorldPosition()
  n131 = Custom!(
   TotalDistance
   LightDirection=@MaterialExpressionSkyAtmosphereLightDirection_2
   WorldPositionBehindTranslucent=@MaterialExpressionMaterialFunctionCall_0.Result
   CameraDirection=@MaterialExpressionMaterialFunctionCall_11.Result
   CameraPosition=@MaterialExpressionCameraPositionWS_1
   WTU_A=@MaterialExpressionAppendVector_4
   WTU_B=@MaterialExpressionAppendVector_5
   WTU_C=@MaterialExpressionAppendVector_6
   WTU_D=@MaterialExpressionAppendVector_7
   WorldPosition=@MaterialExpressionWorldPosition_3.XYZ
  ) {
   code="    float3 RayMarchingDirection = TotalDistance*LightDirection;\r\n    float3 RayMarchingPosition = WorldPositionBehindTranslucent+RayMarchingDirection;\r\n    float3 ProjectedRayMarchingDirection=(RayMarchingDirection-dot(RayMarchingDirection,CameraDirection)*CameraDirection)*(WorldPosition-CameraPosition)/(WorldPositionBehindTranslucent-CameraPosition);\r\n    float4x4 WorldToUnit = {WTU_A,WTU_B,WTU_C,WTU_D};\r\n    float4 TranslatedRayMarchPos = mul(float4(WorldPosition+ProjectedRayMarchingDirection,1.0),WorldToUnit);\r\n    float2 NewUV = TranslatedRayMarchPos.xy;\r\n\treturn NewUV;"
   outputType=CMOT_Float2
   inputs [
    {inputName=TotalDistance, input {expression=null}}
    {
     inputName=LightDirection
     input {expression=&self:MaterialExpressionSkyAtmosphereLightDirection_2}
    }
    {
     inputName=WorldPositionBehindTranslucent
     input {expression=&self:MaterialExpressionMaterialFunctionCall_0}
    }
    {
     inputName=CameraDirection
     input {expression=&self:MaterialExpressionMaterialFunctionCall_11}
    }
    {inputName=CameraPosition, input {expression=&self:MaterialExpressionCameraPositionWS_1}}
    {inputName=WTU_A, input {expression=&self:MaterialExpressionAppendVector_4}}
    {inputName=WTU_B, input {expression=&self:MaterialExpressionAppendVector_5}}
    {inputName=WTU_C, input {expression=&self:MaterialExpressionAppendVector_6}}
    {inputName=WTU_D, input {expression=&self:MaterialExpressionAppendVector_7}}
    {
     inputName=WorldPosition
     input {expression=&self:MaterialExpressionWorldPosition_3, mask=1, maskB=1, maskG=1, maskR=1}
    }
   ]
  }
  TextureSample(UVs=@n131)
 ]
}
